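# Physics asset for the M_MED_Blue_Flames male medium player body cosmetic, covering four dangling grenade accessories, front and back on each side.
# Each grenade capsule is constrained to a tiny MSTR anchor body with tight swing and twist limits, and each pin hangs off its grenade with looser limits.
# Unconstrained thigh capsules are also present, likely as colliders for the swinging grenades.
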PhysicsAsset /BRCosmetics/Characters/Player/Male/Medium/Bodies/M_MED_Blue_Flames/Meshes/M_MED_Blue_Flames_Physics
bodies {
 dyn_grenade_bk_01_l [{shapeType=Capsule, length=5.75, radius=5, center=<1.0123 -0.0006 3.4159>, rotation=<rot -90 -0.005 0.005>}]
 dyn_grenade_bk_01_l_MSTR [{shapeType=Capsule, length=0.01, radius=0.01}]
 dyn_grenade_bk_01_r [{shapeType=Capsule, length=5.75, radius=5, center=<1.0123 -0.0006 3.4159>, rotation=<rot -90 -0.005 0.005>}]
 dyn_grenade_bk_01_r_MSTR [{shapeType=Capsule, length=0.01, radius=0.01}]
 dyn_grenade_fr_01_l [{shapeType=Capsule, length=5.75, radius=5, center=<1.0123 -0.0006 3.4159>, rotation=<rot -90 -0.005 0.005>}]
 dyn_grenade_fr_01_l_MSTR [{shapeType=Capsule, length=0.01, radius=0.01}]
 dyn_grenade_fr_01_r [{shapeType=Capsule, length=5.75, radius=5, center=<1.0123 -0.0006 3.4159>, rotation=<rot -90 -0.005 0.005>}]
 dyn_grenade_fr_01_r_MSTR [{shapeType=Capsule, length=0.01, radius=0.01}]
 dyn_grenadepin_bk_01_l [{shapeType=Capsule, length=1.5866, radius=1.5866, center=<0.9888 0 0>}]
 dyn_grenadepin_bk_01_r [{shapeType=Capsule, length=1.5866, radius=1.5866}]
 dyn_grenadepin_fr_01_l [{shapeType=Capsule, length=1.5866, radius=1.5866}]
 dyn_grenadepin_fr_01_r [{shapeType=Capsule, length=1.5866, radius=1.5866}]
 thigh_l [{shapeType=Capsule, length=24.4311, radius=6.9153, center=<-13.436 0.0003 -0.0003>, rotation=<rot 90 0 0>}]
 thigh_r [{shapeType=Capsule, length=24.4311, radius=6.9153, center=<13.436 -0.0003 0.0003>, rotation=<rot 90 0 -180>}]
}
constraints [
 {bone1Name=dyn_grenade_bk_01_l, bone2Name=dyn_grenade_bk_01_l_MSTR, swing1LimitDegrees=15, swing2LimitDegrees=15, twistLimitDegrees=2}
 {bone1Name=dyn_grenade_fr_01_l, bone2Name=dyn_grenade_fr_01_l_MSTR, swing1LimitDegrees=15, swing2LimitDegrees=15, twistLimitDegrees=2}
 {bone1Name=dyn_grenade_bk_01_r, bone2Name=dyn_grenade_bk_01_r_MSTR, swing1LimitDegrees=15, swing2LimitDegrees=15, twistLimitDegrees=2}
 {bone1Name=dyn_grenade_fr_01_r, bone2Name=dyn_grenade_fr_01_r_MSTR, swing1LimitDegrees=15, swing2LimitDegrees=15, twistLimitDegrees=2}
 {bone1Name=dyn_grenadepin_bk_01_l, bone2Name=dyn_grenade_bk_01_l, swing1LimitDegrees=35, swing2LimitDegrees=35, twistLimitDegrees=45}
 {bone1Name=dyn_grenadepin_fr_01_l, bone2Name=dyn_grenade_fr_01_l, swing1LimitDegrees=35, swing2LimitDegrees=35, twistLimitDegrees=45}
 {bone1Name=dyn_grenadepin_bk_01_r, bone2Name=dyn_grenade_bk_01_r, swing1LimitDegrees=35, swing2LimitDegrees=35, twistLimitDegrees=45}
 {bone1Name=dyn_grenadepin_fr_01_r, bone2Name=dyn_grenade_fr_01_r, swing1LimitDegrees=35, swing2LimitDegrees=35, twistLimitDegrees=45}
]
