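# Physics asset for the face accessory mesh of the Blue Flames male medium player cosmetic, defining bodies for dynamic hair and tube bones.
# Hair bones use small capsules in one- or two-segment chains, each constrained to a near-zero-size MSTR root body.
# Swing and twist are limited, with tighter swing at the top and center-front roots.
# Tube bones have only near-zero-size capsules and no constraints here.
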PhysicsAsset /BRCosmetics/Characters/Player/Male/Medium/Bodies/M_MED_Blue_Flames/Meshes/Parts/M_MED_Blue_Flames_FaceAcc_Physics
bodies {
 dyn_hair_01_r [{shapeType=Capsule, length=2.5, radius=1.5, center=<1.6706 -0.0002 0>, rotation=<rot 90 -0.0033 -0.0032>}]
 dyn_hair_01_r_MSTR [{shapeType=Capsule, length=0.01, radius=0.01}]
 dyn_hair_bk_01 [{shapeType=Capsule, length=2.5, radius=1.5, center=<1.6706 -0.0002 0>, rotation=<rot 90 -0.0033 -0.0032>}]
 dyn_hair_bk_01_MSTR [{shapeType=Capsule, length=0.01, radius=0.01}]
 dyn_hair_bk_01_l [{shapeType=Capsule, length=2.5, radius=1.5, center=<1.6706 -0.0002 0>, rotation=<rot 90 -0.0033 -0.0032>}]
 dyn_hair_bk_01_l_MSTR [{shapeType=Capsule, length=0.01, radius=0.01}]
 dyn_hair_bk_02 [{shapeType=Capsule, length=2.5, radius=1.5, center=<0.2088 -0.0002 0>, rotation=<rot 90 0.0149 0.0149>}]
 dyn_hair_fr_01 [{shapeType=Capsule, length=2.5, radius=1.5, center=<0.2088 -0.0002 0>, rotation=<rot 90 0.0149 0.0149>}]
 dyn_hair_fr_01_MSTR [{shapeType=Capsule, length=0.01, radius=0.01}]
 dyn_hair_fr_01_l [{shapeType=Capsule, length=2.5, radius=1.5, center=<1.6706 -0.0002 0>, rotation=<rot 90 -0.0033 -0.0032>}]
 dyn_hair_fr_01_l_MSTR [{shapeType=Capsule, length=0.01, radius=0.01}]
 dyn_hair_fr_01_r [{shapeType=Capsule, length=2.5, radius=1.5, center=<1.6706 -0.0002 0>, rotation=<rot 90 -0.0033 -0.0032>}]
 dyn_hair_fr_01_r_MSTR [{shapeType=Capsule, length=0.01, radius=0.01}]
 dyn_hair_fr_02 [{shapeType=Capsule, length=2.5, radius=1.5, center=<1.86 -0.0002 -0.0001>, rotation=<rot 90 179.9802 179.9803>}]
 dyn_hair_mid_01 [{shapeType=Capsule, length=2.5, radius=1.5, center=<1.6706 -0.0002 0>, rotation=<rot 90 -0.0033 -0.0032>}]
 dyn_hair_mid_01_MSTR [{shapeType=Capsule, length=0.01, radius=0.01}]
 dyn_hair_mid_01_l [{shapeType=Capsule, length=2.5, radius=1.5, center=<1.6706 -0.0002 0>, rotation=<rot 90 -0.0033 -0.0032>}]
 dyn_hair_mid_01_l_MSTR [{shapeType=Capsule, length=0.01, radius=0.01}]
 dyn_hair_mid_02_l [{shapeType=Capsule, length=2.5, radius=1.5, center=<0.2088 -0.0002 0>, rotation=<rot 90 0.0149 0.0149>}]
 dyn_hair_tp_01 [{shapeType=Capsule, length=2.5, radius=1.5, center=<1.6706 -0.0002 0>, rotation=<rot 90 -0.0033 -0.0032>}]
 dyn_hair_tp_01_MSTR [{shapeType=Capsule, length=0.01, radius=0.01}]
 dyn_hair_tp_02 [{shapeType=Capsule, length=2.5, radius=1.5, center=<0.2088 -0.0002 0>, rotation=<rot 90 0.0149 0.0149>}]
 dyn_tube_01_l [{shapeType=Capsule, length=0.01, radius=0.01}]
 dyn_tube_01_r [{shapeType=Capsule, length=0.01, radius=0.01}]
 dyn_tube_bk_01_l [{shapeType=Capsule, length=0.01, radius=0.01}]
 dyn_tube_bk_01_r [{shapeType=Capsule, length=0.01, radius=0.01}]
 dyn_tube_spine_01_l [{shapeType=Capsule, length=0.01, radius=0.01}]
 dyn_tube_spine_01_r [{shapeType=Capsule, length=0.01, radius=0.01}]
 dyn_tubehead_01_l [{shapeType=Capsule, length=0.01, radius=0.01}]
 dyn_tubehead_01_r [{shapeType=Capsule, length=0.01, radius=0.01}]
}
constraints [
 {bone1Name=dyn_hair_bk_01, bone2Name=dyn_hair_bk_01_MSTR, swing1LimitDegrees=30, swing2LimitDegrees=30, twistLimitDegrees=5}
 {bone1Name=dyn_hair_bk_02, bone2Name=dyn_hair_bk_01, swing1LimitDegrees=30, swing2LimitDegrees=30, twistLimitDegrees=5}
 {bone1Name=dyn_hair_mid_01, bone2Name=dyn_hair_mid_01_MSTR, swing1LimitDegrees=30, swing2LimitDegrees=30, twistLimitDegrees=5}
 {bone1Name=dyn_hair_tp_01, bone2Name=dyn_hair_tp_01_MSTR, swing1LimitDegrees=15, swing2LimitDegrees=15, twistLimitDegrees=5}
 {bone1Name=dyn_hair_tp_02, bone2Name=dyn_hair_tp_01, swing1LimitDegrees=30, swing2LimitDegrees=30, twistLimitDegrees=5}
 {bone1Name=dyn_hair_fr_01, bone2Name=dyn_hair_fr_01_MSTR, swing1LimitDegrees=15, swing2LimitDegrees=15, twistLimitDegrees=5}
 {bone1Name=dyn_hair_fr_02, bone2Name=dyn_hair_fr_01, swing1LimitDegrees=30, swing2LimitDegrees=30, twistLimitDegrees=5}
 {bone1Name=dyn_hair_fr_01_r, bone2Name=dyn_hair_fr_01_r_MSTR, swing1LimitDegrees=30, swing2LimitDegrees=30, twistLimitDegrees=5}
 {bone1Name=dyn_hair_01_r, bone2Name=dyn_hair_01_r_MSTR, swing1LimitDegrees=30, swing2LimitDegrees=30, twistLimitDegrees=5}
 {bone1Name=dyn_hair_fr_01_l, bone2Name=dyn_hair_fr_01_l_MSTR, swing1LimitDegrees=30, swing2LimitDegrees=30, twistLimitDegrees=5}
 {bone1Name=dyn_hair_bk_01_l, bone2Name=dyn_hair_bk_01_l_MSTR, swing1LimitDegrees=30, swing2LimitDegrees=30, twistLimitDegrees=5}
 {bone1Name=dyn_hair_mid_01_l, bone2Name=dyn_hair_mid_01_l_MSTR, swing1LimitDegrees=30, swing2LimitDegrees=30, twistLimitDegrees=5}
 {bone1Name=dyn_hair_mid_02_l, bone2Name=dyn_hair_mid_01_l, swing1LimitDegrees=30, swing2LimitDegrees=30, twistLimitDegrees=5}
]
