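# Post-pose AnimBP for the Blue Flames FaceAcc part, animating its left and right tubes on the parent mesh's pose.
# Weighted bone constraints feed AnimDynamics, additive bone offsets and LookAts, then RigidBody simulation.
# EventGraph maps tube-head-to-body distances to offset alphas and swaps RigidBody gravity for scaled skydiving force while skydiving.
# The right side writes only its inverse alpha, zeroed at close range; RmouthToTubeDistanceAlpha is never set here.
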AnimBlueprint /BRCosmetics/Characters/Player/Male/Medium/Bodies/M_MED_Blue_Flames/Meshes/Parts/M_MED_Blue_Flames_FaceAcc_AnimBP
target_skeleton=&/M_MED_Blue_Flames_FaceAcc_Skeleton
graphs [
 AnimGraph {nodes [
   n1 = "Constraint - dyn_tube_01_l (PPPPPPPPPPPPP)"!(
    ComponentPose=Local To Component(LocalPose=Copy Pose From Mesh() {bUseAttachedParent=true})
    Alpha=1
    ConstraintWeights_0=0
    ConstraintWeights_1:float=0.5952
    ConstraintWeights_2:float=0
    ConstraintWeights_3:float=0
    ConstraintWeights_4:float=0
    ConstraintWeights_5:float=0.0532
    ConstraintWeights_6:float=0.0102
    ConstraintWeights_7:float=0.2369
    ConstraintWeights_8:float=0.0595
    ConstraintWeights_9:float=0.0365
    ConstraintWeights_10:float=0
    ConstraintWeights_11:float=0.0084
    ConstraintWeights_12:float=0
   ) -> (Pose) {
    alphaBoolBlend {blendOption=Linear}
    boneToModify {boneName=dyn_tube_01_l}
    constraintSetup [
     {targetBone {boneName=head}}
     {targetBone {boneName=clavicle_l}}
     {targetBone {boneName=upperarm_twist_02_l}}
     {targetBone {boneName=spine_01}}
     {targetBone {boneName=spine_02}}
     {targetBone {boneName=spine_03}}
     {targetBone {boneName=spine_04}}
     {targetBone {boneName=spine_05}}
     {targetBone {boneName=upperarm_twist_01_l}}
     {targetBone {boneName=neck_02}}
     {targetBone {boneName=pelvis}}
     {targetBone {boneName=neck_01}}
     {targetBone {boneName=upperarm_l}}
    ]
    constraintWeights [1,1,1,1,1,1,1,1,1,1,1,1,1]
   }
   n2 = "Constraint - dyn_tube_01_r (PPPPPPPPPPPPP)"!(
    ComponentPose=@n1
    Alpha=1
    ConstraintWeights_0=0
    ConstraintWeights_1:float=0
    ConstraintWeights_2:float=0.0708
    ConstraintWeights_3:float=0
    ConstraintWeights_4:float=0
    ConstraintWeights_5:float=0
    ConstraintWeights_6:float=0.0464
    ConstraintWeights_7:float=0.0117
    ConstraintWeights_8:float=0.2207
    ConstraintWeights_9:float=0.0347
    ConstraintWeights_10:float=0.6081
    ConstraintWeights_11:float=0
    ConstraintWeights_12:float=0.0076
   ) -> (Pose) {
    alphaBoolBlend {blendOption=Linear}
    boneToModify {boneName=dyn_tube_01_r}
    constraintSetup [
     {targetBone {boneName=head}}
     {targetBone {boneName=upperarm_r}}
     {targetBone {boneName=upperarm_twist_01_r}}
     {targetBone {boneName=upperarm_twist_02_r}}
     {targetBone {boneName=spine_01}}
     {targetBone {boneName=spine_02}}
     {targetBone {boneName=spine_03}}
     {targetBone {boneName=spine_04}}
     {targetBone {boneName=spine_05}}
     {targetBone {boneName=neck_02}}
     {targetBone {boneName=clavicle_r}}
     {targetBone {boneName=pelvis}}
     {targetBone {boneName=neck_01}}
    ]
    constraintWeights [1,1,1,1,1,1,1,1,1,1,1,1,1]
   }
   n3 = "Constraint - dyn_tube_bk_01_l (PPPPPPPPPPPP)"!(
    ComponentPose=@n2
    Alpha=1
    ConstraintWeights_0=0
    ConstraintWeights_1:float=0.4235
    ConstraintWeights_2:float=0
    ConstraintWeights_3:float=0
    ConstraintWeights_4:float=0
    ConstraintWeights_5:float=0.1167
    ConstraintWeights_6:float=0.0735
    ConstraintWeights_7:float=0.373
    ConstraintWeights_8:float=0.0033
    ConstraintWeights_9:float=0.01
    ConstraintWeights_10:float=0
    ConstraintWeights_11:float=0
   ) -> (Pose) {
    alphaBoolBlend {blendOption=Linear}
    boneToModify {boneName=dyn_tube_bk_01_l}
    constraintSetup [
     {targetBone {boneName=head}}
     {targetBone {boneName=clavicle_l}}
     {targetBone {boneName=upperarm_twist_02_l}}
     {targetBone {boneName=spine_01}}
     {targetBone {boneName=spine_02}}
     {targetBone {boneName=spine_03}}
     {targetBone {boneName=spine_04}}
     {targetBone {boneName=spine_05}}
     {targetBone {boneName=upperarm_twist_01_l}}
     {targetBone {boneName=neck_02}}
     {targetBone {boneName=neck_01}}
     {targetBone {boneName=upperarm_l}}
    ]
    constraintWeights [1,1,1,1,1,1,1,1,1,1,1,1]
   }
   n4 = "Constraint - dyn_tube_bk_01_r (PPPPPPPPPPPPPP)"!(
    ComponentPose=@n3
    Alpha=1
    ConstraintWeights_0=0
    ConstraintWeights_1:float=0.0005
    ConstraintWeights_2:float=0
    ConstraintWeights_3:float=0.0436
    ConstraintWeights_4:float=0.0003
    ConstraintWeights_5:float=0
    ConstraintWeights_6:float=0.0003
    ConstraintWeights_7:float=0.1145
    ConstraintWeights_8:float=0.095
    ConstraintWeights_9:float=0.2989
    ConstraintWeights_10:float=0.008
    ConstraintWeights_11:float=0.4388
    ConstraintWeights_12:float=0
    ConstraintWeights_13:float=0.0001
   ) -> (Pose) {
    alphaBoolBlend {blendOption=Linear}
    boneToModify {boneName=dyn_tube_bk_01_r}
    constraintSetup [
     {targetBone {boneName=head}}
     {targetBone {boneName=upperarm_r}}
     {targetBone {boneName=lowerarm_r}}
     {targetBone {boneName=upperarm_twist_01_r}}
     {targetBone {boneName=upperarm_twist_02_r}}
     {targetBone {boneName=spine_01}}
     {targetBone {boneName=spine_02}}
     {targetBone {boneName=spine_03}}
     {targetBone {boneName=spine_04}}
     {targetBone {boneName=spine_05}}
     {targetBone {boneName=neck_02}}
     {targetBone {boneName=clavicle_r}}
     {targetBone {boneName=pelvis}}
     {targetBone {boneName=neck_01}}
    ]
    constraintWeights [1,1,1,1,1,1,1,1,1,1,1,1,1,1]
   }
   n5 = Anim Dynamics - Bone: dyn_tube_01_l(ComponentPose=@n4, Alpha=1, ExternalForce="0,0,0") {
    angularDampingOverride=1
    angularSpringConstant=1500
    bAngularSpring=true
    bGravityOverrideInSimSpace=true
    bOverrideAngularDamping=true
    gravityScale=0.35
    simulationSpace=RootRelative
    alphaBoolBlend {blendOption=Linear}
    boundBone {boneName=dyn_tube_01_l}
    componentLinearAccScale=<0.3 0.3 0.7>
    componentLinearVelScale=<0.3 0.3 0.7>
    physicsBodyDefinitions [{
      sphereCollisionRadius=0
      boundBone {boneName=dyn_tube_01_l}
      boxExtents=<10 10 10>
      constraintSetup {angularLimitsMax=<0 0 0>, angularLimitsMin=<0 -20 0>, angularTarget=<0 0 0>}
      localJointOffset=<2 0 0>
     }]
   }
   n6 = Anim Dynamics - Bone: dyn_tubehead_01_l(ComponentPose=@n5, Alpha=1, ExternalForce="0,0,0") {
    angularDampingOverride=1
    angularSpringConstant=1500
    bAngularSpring=true
    bGravityOverrideInSimSpace=true
    bOverrideAngularDamping=true
    gravityScale=0.35
    simulationSpace=RootRelative
    alphaBoolBlend {blendOption=Linear}
    boundBone {boneName=dyn_tubehead_01_l}
    componentLinearAccScale=<0.3 0.3 0.3>
    componentLinearVelScale=<0.3 0.3 0.3>
    physicsBodyDefinitions [{
      sphereCollisionRadius=0
      boundBone {boneName=dyn_tubehead_01_l}
      boxExtents=<10 10 10>
      constraintSetup {angularLimitsMax=<0 10 0>, angularLimitsMin=<0 -10 0>, angularTarget=<0 0 0>}
      localJointOffset=<2 0 0>
     }]
   }
   n7 = Anim Dynamics - Bone: dyn_tube_bk_01_l(ComponentPose=@n6, Alpha=1, ExternalForce="0,0,0") {
    angularDampingOverride=1
    angularSpringConstant=1500
    bAngularSpring=true
    bGravityOverrideInSimSpace=true
    bOverrideAngularDamping=true
    gravityScale=0.35
    simulationSpace=RootRelative
    alphaBoolBlend {blendOption=Linear}
    boundBone {boneName=dyn_tube_bk_01_l}
    componentLinearAccScale=<0.3 0.3 0.7>
    componentLinearVelScale=<0.3 0.3 0.7>
    physicsBodyDefinitions [{
      sphereCollisionRadius=0
      boundBone {boneName=dyn_tube_bk_01_l}
      boxExtents=<10 10 10>
      constraintSetup {angularLimitsMax=<0 10 0>, angularLimitsMin=<0 -10 0>, angularTarget=<0 0 0>}
      localJointOffset=<2 0 0>
     }]
   }
   n8 = Anim Dynamics - Bone: dyn_tube_spine_01_l(ComponentPose=@n7, Alpha=1, ExternalForce="0,0,0") {
    angularDampingOverride=1
    angularSpringConstant=1500
    bAngularSpring=true
    bGravityOverrideInSimSpace=true
    bOverrideAngularDamping=true
    gravityScale=0.35
    simulationSpace=RootRelative
    alphaBoolBlend {blendOption=Linear}
    boundBone {boneName=dyn_tube_spine_01_l}
    componentLinearAccScale=<0.3 0.3 0.7>
    componentLinearVelScale=<0.3 0.3 0.7>
    physicsBodyDefinitions [{
      sphereCollisionRadius=0
      boundBone {boneName=dyn_tube_spine_01_l}
      boxExtents=<10 10 10>
      constraintSetup {angularLimitsMax=<0 10 10>, angularLimitsMin=<0 -10 -10>, angularTarget=<0 0 0>}
      localJointOffset=<2 0 0>
     }]
   }
   n9 = Get LmouthToTubeDistanceAlpha() -> (LmouthToTubeDistanceAlpha:double)
   n10 = "Transform (Modify) Bone - Bone: dyn_tube_02_l"(ComponentPose=@n8, Alpha=@n9, Translation="2,0,0", Rotation="0,0,0", Scale="1,1,1") {
    rotationMode=BMM_Additive
    rotationSpace=BCS_BoneSpace
    scaleMode=BMM_Additive
    scaleSpace=BCS_BoneSpace
    translationMode=BMM_Additive
    translationSpace=BCS_BoneSpace
    alphaBoolBlend {blendOption=Linear}
    boneToModify {boneName=dyn_tube_02_l}
    translation {x=2}
   }
   n11 = "Transform (Modify) Bone - Bone: dyn_tubehead_02_l"(ComponentPose=@n10, Alpha=@n9, Translation="-2,0,0", Rotation="0,0,0", Scale="1,1,1") {
    rotationMode=BMM_Additive
    rotationSpace=BCS_BoneSpace
    scaleMode=BMM_Additive
    scaleSpace=BCS_BoneSpace
    translationMode=BMM_Additive
    translationSpace=BCS_BoneSpace
    alphaBoolBlend {blendOption=Linear}
    boneToModify {boneName=dyn_tubehead_02_l}
    translation {x=-2}
   }
   n12 = "Transform (Modify) Bone - Bone: dyn_tubehead_00_l"(ComponentPose=@n11, Alpha=@n9, Translation="0,0,0", Rotation="-30,0.000,0.000", Scale="1,1,1") {
    rotationMode=BMM_Additive
    rotationSpace=BCS_BoneSpace
    scaleMode=BMM_Additive
    scaleSpace=BCS_BoneSpace
    translationMode=BMM_Additive
    translationSpace=BCS_BoneSpace
    alphaBoolBlend {blendOption=Linear}
    boneToModify {boneName=dyn_tubehead_00_l}
    rotation {pitch=-30}
   }
   n13 = "Transform (Modify) Bone - Bone: dyn_tubehead_00_l"(
    ComponentPose=@n12
    Alpha=Get LmouthToTubeDistanceAlpha_inv() -> (LmouthToTubeDistanceAlpha_inv:double)
    Translation="0,0,0"
    Rotation="15,0.000,0.000"
    Scale="1,1,1"
   ) {
    rotationMode=BMM_Additive
    rotationSpace=BCS_BoneSpace
    scaleMode=BMM_Additive
    scaleSpace=BCS_BoneSpace
    translationMode=BMM_Additive
    translationSpace=BCS_BoneSpace
    alphaBoolBlend {blendOption=Linear}
    boneToModify {boneName=dyn_tubehead_00_l}
    rotation {pitch=15}
   }
   n14 = "Transform (Modify) Bone - Bone: dyn_tubehead_01_l"(ComponentPose=@n13, Alpha=@n9, Translation="0,0,0", Rotation="1,0.000,0.000", Scale="1,1,1") {
    rotationMode=BMM_Additive
    rotationSpace=BCS_BoneSpace
    scaleMode=BMM_Additive
    scaleSpace=BCS_BoneSpace
    translationMode=BMM_Additive
    translationSpace=BCS_BoneSpace
    alphaBoolBlend {blendOption=Linear}
    boneToModify {boneName=dyn_tubehead_01_l}
    rotation {pitch=1}
   }
   n15 = Look At - Bone: dyn_tubehead_01_l(ComponentPose=@n14, Alpha=1) {
    alphaBoolBlend {blendOption=Linear}
    boneToModify {boneName=dyn_tubehead_01_l}
    lookAtLocation {x=0}
    lookAtTarget {boneReference {boneName=dyn_tube_02_l}}
    lookAt_Axis {axis=<1 0>}
    lookUp_Axis {axis {x=0, z=1}}
   }
   n16 = Anim Dynamics - Bone: dyn_tube_01_r(ComponentPose=@n15, Alpha=1, ExternalForce="0,0,0") {
    angularDampingOverride=1
    angularSpringConstant=1500
    bAngularSpring=true
    bGravityOverrideInSimSpace=true
    bOverrideAngularDamping=true
    gravityScale=0.35
    simulationSpace=RootRelative
    alphaBoolBlend {blendOption=Linear}
    boundBone {boneName=dyn_tube_01_r}
    componentLinearAccScale=<0.3 0.3 0.7>
    componentLinearVelScale=<0.3 0.3 0.7>
    physicsBodyDefinitions [{
      sphereCollisionRadius=0
      boundBone {boneName=dyn_tube_01_r}
      boxExtents=<10 10 10>
      constraintSetup {angularLimitsMax=<0 5 0>, angularLimitsMin=<0 -20 0>, angularTarget=<0 0 0>}
      localJointOffset=<2 0 0>
     }]
   }
   n17 = Anim Dynamics - Bone: dyn_tubehead_01_r(ComponentPose=@n16, Alpha=1, ExternalForce="0,0,0") {
    angularDampingOverride=1
    angularSpringConstant=1500
    bAngularSpring=true
    bGravityOverrideInSimSpace=true
    bOverrideAngularDamping=true
    gravityScale=0.35
    simulationSpace=RootRelative
    alphaBoolBlend {blendOption=Linear}
    boundBone {boneName=dyn_tubehead_01_r}
    componentLinearAccScale=<0.3 0.3 0.3>
    componentLinearVelScale=<0.3 0.3 0.3>
    physicsBodyDefinitions [{
      sphereCollisionRadius=0
      boundBone {boneName=dyn_tubehead_01_r}
      boxExtents=<10 10 10>
      constraintSetup {angularLimitsMax=<0 10 0>, angularLimitsMin=<0 -10 0>, angularTarget=<0 0 0>}
      localJointOffset=<2 0 0>
     }]
   }
   n18 = Anim Dynamics - Bone: dyn_tube_bk_01_r(ComponentPose=@n17, Alpha=1, ExternalForce="0,0,0") {
    angularDampingOverride=1
    angularSpringConstant=1500
    bAngularSpring=true
    bGravityOverrideInSimSpace=true
    bOverrideAngularDamping=true
    gravityScale=0.35
    simulationSpace=RootRelative
    alphaBoolBlend {blendOption=Linear}
    boundBone {boneName=dyn_tube_bk_01_r}
    componentLinearAccScale=<0.3 0.3 0.7>
    componentLinearVelScale=<0.3 0.3 0.7>
    physicsBodyDefinitions [{
      sphereCollisionRadius=0
      boundBone {boneName=dyn_tube_bk_01_r}
      boxExtents=<10 10 10>
      constraintSetup {angularLimitsMax=<0 10 0>, angularLimitsMin=<0 -10 0>, angularTarget=<0 0 0>}
      localJointOffset=<2 0 0>
     }]
   }
   n19 = Anim Dynamics - Bone: dyn_tube_spine_01_r(ComponentPose=@n18, Alpha=1, ExternalForce="0,0,0") {
    angularDampingOverride=1
    angularSpringConstant=1500
    bAngularSpring=true
    bGravityOverrideInSimSpace=true
    bOverrideAngularDamping=true
    gravityScale=0.35
    simulationSpace=RootRelative
    alphaBoolBlend {blendOption=Linear}
    boundBone {boneName=dyn_tube_spine_01_r}
    componentLinearAccScale=<0.3 0.3 0.5>
    componentLinearVelScale=<0.7 0.7 0.7>
    physicsBodyDefinitions [{
      sphereCollisionRadius=0
      boundBone {boneName=dyn_tube_spine_01_r}
      boxExtents=<10 10 10>
      constraintSetup {angularLimitsMax=<0 10 10>, angularLimitsMin=<0 -10 -10>, angularTarget=<0 0 0>}
      localJointOffset=<2 0 0>
     }]
   }
   n20 = Get RmouthToTubeDistanceAlpha() -> (RmouthToTubeDistanceAlpha:double)
   n21 = "Transform (Modify) Bone - Bone: dyn_tube_02_r"(ComponentPose=@n19, Alpha=@n20, Translation="2,0,0", Rotation="0,0,0", Scale="1,1,1") {
    rotationMode=BMM_Additive
    rotationSpace=BCS_BoneSpace
    scaleMode=BMM_Additive
    scaleSpace=BCS_BoneSpace
    translationMode=BMM_Additive
    translationSpace=BCS_BoneSpace
    alphaBoolBlend {blendOption=Linear}
    boneToModify {boneName=dyn_tube_02_r}
    translation {x=2}
   }
   n22 = "Transform (Modify) Bone - Bone: dyn_tubehead_02_r"(ComponentPose=@n21, Alpha=@n20, Translation="-2,0,0", Rotation="0,0,0", Scale="1,1,1") {
    rotationMode=BMM_Additive
    rotationSpace=BCS_BoneSpace
    scaleMode=BMM_Additive
    scaleSpace=BCS_BoneSpace
    translationMode=BMM_Additive
    translationSpace=BCS_BoneSpace
    alphaBoolBlend {blendOption=Linear}
    boneToModify {boneName=dyn_tubehead_02_r}
    translation {x=-2}
   }
   n23 = "Transform (Modify) Bone - Bone: dyn_tubehead_00_r"(ComponentPose=@n22, Alpha=@n20, Translation="0,0,0", Rotation="-60,0.000,0.000", Scale="1,1,1") {
    rotationMode=BMM_Additive
    rotationSpace=BCS_BoneSpace
    scaleMode=BMM_Additive
    scaleSpace=BCS_BoneSpace
    translationMode=BMM_Additive
    translationSpace=BCS_BoneSpace
    alphaBoolBlend {blendOption=Linear}
    boneToModify {boneName=dyn_tubehead_00_r}
    rotation {pitch=-60}
   }
   n24 = "Transform (Modify) Bone - Bone: dyn_tubehead_00_r"(
    ComponentPose=@n23
    Alpha=Get RmouthToTubeDistanceAlpha_inv() -> (RmouthToTubeDistanceAlpha_inv:double)
    Translation="0,0,0"
    Rotation="15,0.000,0.000"
    Scale="1,1,1"
   ) {
    rotationMode=BMM_Additive
    rotationSpace=BCS_BoneSpace
    scaleMode=BMM_Additive
    scaleSpace=BCS_BoneSpace
    translationMode=BMM_Additive
    translationSpace=BCS_BoneSpace
    alphaBoolBlend {blendOption=Linear}
    boneToModify {boneName=dyn_tubehead_00_r}
    rotation {pitch=15}
   }
   n25 = "Transform (Modify) Bone - Bone: dyn_tubehead_01_r"(ComponentPose=@n24, Alpha=@n20, Translation="0,0,0", Rotation="1,0.000,0.000", Scale="1,1,1") {
    rotationMode=BMM_Additive
    rotationSpace=BCS_BoneSpace
    scaleMode=BMM_Additive
    scaleSpace=BCS_BoneSpace
    translationMode=BMM_Additive
    translationSpace=BCS_BoneSpace
    alphaBoolBlend {blendOption=Linear}
    boneToModify {boneName=dyn_tubehead_01_r}
    rotation {pitch=1}
   }
   n26 = Look At - Bone: dyn_tubehead_01_r(ComponentPose=@n25, Alpha=1) {
    alphaBoolBlend {blendOption=Linear}
    boneToModify {boneName=dyn_tubehead_01_r}
    lookAtLocation {x=0}
    lookAtTarget {boneReference {boneName=dyn_tube_02_r}}
    lookAt_Axis {axis=<1 0>}
    lookUp_Axis {axis {x=0, z=1}}
   }
   n27 = RigidBody!(ComponentPose=@n26, Alpha, OverrideWorldGravity:Vector=Get windVector() -> (windVector:Vector), ExternalForce) -> (Pose) {
    bClampLinearTranslationLimitToRefPose=true
    bOverrideWorldGravity=true
    overridePhysicsAsset=&/M_MED_Blue_Flames_FaceAcc_Physics
    simulationSpace=BaseBoneSpace
    alphaBoolBlend {blendOption=Linear}
    baseBoneRef {boneName=root}
    componentAppliedLinearAccClamp=<3000 3000 3000>
    componentLinearAccScale=<0.3 0.3 0.7>
    componentLinearVelScale=<0.7 0.7 0.7>
    overrideWorldGravity {z=-980}
   }
   Output Pose(Result=Component To Local(ComponentPose=@n27)) {layerGroup=None}
  ]}
 EventGraph {nodes [
   Event BlueprintUpdateAnimation()
   > n1 = Sequence!(execute) -> (then_0, then_1, then_2:Exec, then_3:Exec, then_4:Exec, then_5:Exec, then_6:Exec)
   n2 = Break Fort Anim Input Skydiving External Force(FortAnimInput_SkydivingExternalForce=Get SkydivingExternalForceInput()) -> (
    bUseSkydivingVectorForce=~
    SkydivingVectorForce=~
    HeadToPelvisDirection=~
    FloatingMultiplier=~
    FloatingAdditive=~
    DivingMultiplier=~
    DivingAdditive=~
    ParachutingMultiplier=~
    ParachutingAdditive=~
   )
   > Set windVector(windVector:Vector=SelectVector(A=vector * float(A=@n2.SkydivingVectorForce, B=0.15), B="0, 0,-980", bPickA=Get bIsSkyDiving())) -> (Output_Get:Vector)
   n3 = GetPartSkeletalMeshComponent()
   n4 = TransformFromBoneSpace!(
    execute=@n1.then_1
    self:Skinned Mesh Component Object Reference=@n3
    BoneName:Name=dyn_tubehead_00_l
    InPosition:Vector="0, 0, 0"
    InRotation:Rotator="0, 0, 0"
   ) -> (then, OutPosition:Vector, OutRotation:Rotator)
   n5 = TransformFromBoneSpace(execute=@n1.then_2, self=@n3, BoneName=dyn_tubebody_01_l, InPosition="0, 0, 0", InRotation="0, 0, 0")
   n6 = "Distance (Vector)"!(V1:Vector=@n4, V2:Vector=@n5) -> (ReturnValue:double)
   n7 = MapRangeClamped(Value=@n6, InRangeA=5, InRangeB=10, OutRangeA=1, OutRangeB=0)
   Set LmouthToTubeDistanceAlpha(execute=@n1.then_3, LmouthToTubeDistanceAlpha:double=@n7) -> (Output_Get:double)
   > Set LmouthToTubeDistanceAlpha_inv(LmouthToTubeDistanceAlpha_inv:double=float - float(B=@n7) {is=title K2Node_CallFunction}) -> (Output_Get:double)
   n8 = GetPartSkeletalMeshComponent()
   n9 = TransformFromBoneSpace(execute=@n1.then_4, self=@n8, BoneName=dyn_tubehead_00_r, InPosition="0, 0, 0", InRotation="0, 0, 0")
   n10 = TransformFromBoneSpace(execute=@n1.then_5, self=@n8, BoneName=dyn_tubebody_01_r, InPosition="0, 0, 0", InRotation="0, 0, 0")
   n11 = "Distance (Vector)"(V1=@n9, V2=@n10)
   n12 = MapRangeClamped(Value=@n11, InRangeA=5, InRangeB=10, OutRangeA=1, OutRangeB=0)
   Set RmouthToTubeDistanceAlpha_inv(execute=@n1.then_6, RmouthToTubeDistanceAlpha_inv:double=@n12) -> (Output_Get:double)
   > Set RmouthToTubeDistanceAlpha_inv(RmouthToTubeDistanceAlpha_inv=SelectFloat(A=float - float(B=@n12), B=0, bPickA="float > float"(A=@n11, B=7) {is=title K2Node_CallFunction}))
   DrawDebugLine!(
    execute
    LineStart:Vector="0, 0, 0"
    LineEnd:Vector="0, 0, 0"
    LineColor:Linear Color Structure=(R=0.026,G=1,B=0,A=1)
    Duration:float=0
    Thickness:float=1
    DepthPriority:EDrawDebugSceneDepthPriorityGroup Enum=World
   ) -> (then)
   SelectFloat(A=0, B=0, bPickA="float > float"(A=@n6, B=7))
   DrawDebugLine(LineStart="0, 0, 0", LineEnd="0, 0, 0", LineColor=(R=0.026,G=1,B=0,A=1), Duration=0, Thickness=1, DepthPriority=World)
  ]}
]
variables [
 LmouthToTubeDistanceAlpha {type=number}
 LmouthToTubeDistanceAlpha_inv {type=number}
 RmouthToTubeDistanceAlpha {type=number}
 RmouthToTubeDistanceAlpha_inv {type=number}
 windVector {type=Vector}
]
